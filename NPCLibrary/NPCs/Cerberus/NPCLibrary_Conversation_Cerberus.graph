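# Conversation database for the Cerberus NPC, providing its greeting when a player starts talking to it.
# A single speech node picks contextual lines: a first-ever-meeting line, a reaction to a specific outfit, or several lines gated by match conversation count, with a default fallback.
# The greeting then branches through link nodes to shared quest and basic-service conversations and to Cerberus's own dialogue entry.
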ConversationDatabase /NPCLibrary/NPCs/Cerberus/NPCLibrary_Conversation_Cerberus
entry_points [{tag=Conversation.Entry.Cerberus, destinations [#1]}]
nodes [
 ConversationEntryPointNode_7 {
  id=#1
  class=ConversationEntryPointNode
  outputs [#2]
  properties {entryTag {tagName=Conversation.Entry.Cerberus}}
 }
 Speech_C_4 {
  id=#2
  class=Speech_C
  outputs [#3, #4, #5]
  properties {
   nodeName=""
   generalConfig {
    defaultMessage="I AM THE STRONG HEAD OF CERBERUS. WHAT DO YOU WANT, PUP?"
    contextualMessages [
     {
      message="HA HA HA, WHAT WEAKLING CALLS UPON THE HOUNDS OF HADES??"
      contextRequirements [{requirement {class=FortControllerRequirement_IsFirstEverConversationWithNPC}}]
      priority {curve {rowName=Tandem.Speech.Priority.FirstEverInteraction}}
     }
     {
      message="I AM THE STRONG HEAD OF CERBERUS. WHAT DO YOU WANT, PUP?"
      contextRequirements [{requirement {class=FortControllerRequirement_CheckNPCMatchConversationCount, properties {nPCUniqueIdTag {tagName=AiSpawnerData.Type.Tandem.Cerberus}}}}]
      priority {curve {rowName=None}}
     }
     {
      message="Oh good, my other heads are napping. What do you need, hero?"
      contextRequirements [{requirement {class=FortControllerRequirement_CheckNPCMatchConversationCount, properties {nPCUniqueIdTag {tagName=AiSpawnerData.Type.Tandem.Cerberus}}}}]
      priority {curve {rowName=None}}
     }
     {
      message=YOU ARE SMALL AND LOOK TASTY. WHAT DO YOU NEED.
      contextRequirements [{requirement {class=FortControllerRequirement_CheckNPCMatchConversationCount, properties {nPCUniqueIdTag {tagName=AiSpawnerData.Type.Tandem.Cerberus}}}}]
      priority {curve {rowName=None}}
     }
     {
      message="BEHOLD THE UNDERWORLD! READY TO MOVE IN? HA HA HA."
      contextRequirements [{requirement {class=FortControllerRequirement_CheckNPCMatchConversationCount, properties {nPCUniqueIdTag {tagName=AiSpawnerData.Type.Tandem.Cerberus}}}}]
      priority {curve {rowName=None}}
     }
     {
      message=My other heads are... loud and mean. But I will help if I can.
      contextRequirements [{requirement {class=FortControllerRequirement_CheckNPCMatchConversationCount, properties {nPCUniqueIdTag {tagName=AiSpawnerData.Type.Tandem.Cerberus}}}}]
      priority {curve {rowName=None}}
     }
     {
      message="YOU'D BETTER CLEAR OUT BEFORE SHE ARRIVES, OR I MIGHT BITE YA, HA HA HA."
      contextRequirements [{requirement {class=FortControllerRequirement_CheckNPCMatchConversationCount, properties {nPCUniqueIdTag {tagName=AiSpawnerData.Type.Tandem.Cerberus}}}}]
      priority {curve {rowName=None}}
     }
     {
      message="MASTER! BORK BORK BORK!!"
      contextRequirements [{requirement {
         class=FortControllerRequirement_HasCID
         properties {allowedCIds [/BRCosmetics/Athena/Items/Cosmetics/Characters/Character_DarkStance_Inferno]}
        }}]
      priority {curve {rowName=Tandem.Speech.Priority.OutfitReaction}}
     }
    ]
   }
  }
 }
 ConversationLinkNode_2 {id=#5, class=ConversationLinkNode, properties {remoteEntryTag {tagName=Conversation.Entry.Cerberus.Dialogue}}}
 ConversationLinkNode_0 {id=#4, class=ConversationLinkNode, properties {remoteEntryTag {tagName=Conversation.Entry.BasicServices}}}
 ConversationLinkNode_1 {id=#3, class=ConversationLinkNode, properties {remoteEntryTag {tagName=Conversation.Entry.NPCLibrary.Quests}}}
]
