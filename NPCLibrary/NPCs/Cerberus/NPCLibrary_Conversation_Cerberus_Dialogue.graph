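# Conversation database for the Cerberus NPC's dialogue, reached through the Conversation.Entry.Cerberus.Dialogue entry tag.
# An opening "Ask about..." speech branches into choices on the gods, her arrival, a duel, or a Back node.
# Each topic plays a linear chain of three speech lines, and all three converge on one shared ReturnToStart node labelled BACK.
# The Duel branch contains only speech, with no node that starts a fight.
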ConversationDatabase /NPCLibrary/NPCs/Cerberus/NPCLibrary_Conversation_Cerberus_Dialogue
entry_points [{tag=Conversation.Entry.Cerberus.Dialogue, destinations [#1]}]
nodes [
 ConversationEntryPointNode_0 {
  id=#1
  class=ConversationEntryPointNode
  outputs [#2]
  properties {entryTag {tagName=Conversation.Entry.Cerberus.Dialogue}}
 }
 Speech_C_0 {
  id=#2
  class=Speech_C
  outputs [#3, #4, #5, #6]
  properties {nodeName="", softTaskIcon=&T_UI_NPC_question, generalConfig {defaultMessage="WHAT DO YOU WANT, LITTLE PUP??"}}
  sub_nodes [ConversationChoiceNode_5 {class=ConversationChoiceNode, properties {defaultChoiceDisplayText=Ask about...}}]
 }
 Speech_C_4 {
  id=#6
  class=Speech_C
  outputs [#7]
  properties {
   nodeName=""
   softTaskIcon=&T_UI_NPC_question
   generalConfig {defaultMessage=HA HA HA. I WILL BEAT YOU IF YOU DESIRE.}
  }
  sub_nodes [ConversationChoiceNode_15 {class=ConversationChoiceNode, properties {defaultChoiceDisplayText=Duel}}]
 }
 Speech_C_5 {
  id=#7
  class=Speech_C
  outputs [#8]
  properties {
   nodeName=""
   softTaskIcon=&T_UI_NPC_question
   generalConfig {defaultMessage=YOU ARE A LITTLE NOTHING. A TASTY MORSEL.}
  }
 }
 Speech_C_6 {
  id=#8
  class=Speech_C
  outputs [#9]
  properties {
   nodeName=""
   softTaskIcon=&T_UI_NPC_question
   generalConfig {defaultMessage="IT WILL KEEP ME BUSY. I AM SO BORED! BORK BORK."}
  }
 }
 ReturnToStart_C_23 {
  id=#9
  class=ReturnToStart_C
  properties {softTaskIcon=&T_UI_NPC_Back}
  sub_nodes [ConversationChoiceNode_32 {class=ConversationChoiceNode, properties {defaultChoiceDisplayText=BACK}}]
 }
 Speech_C_1 {
  id=#5
  class=Speech_C
  outputs [#10]
  properties {
   nodeName=""
   softTaskIcon=&T_UI_NPC_question
   generalConfig {defaultMessage=SHE WILL ARRIVE ON THE SWEET SUMMER WIND.}
  }
  sub_nodes [ConversationChoiceNode_1 {class=ConversationChoiceNode, properties {defaultChoiceDisplayText=Her Arrival}}]
 }
 Speech_C_2 {
  id=#10
  class=Speech_C
  outputs [#11]
  properties {
   nodeName=""
   softTaskIcon=&T_UI_NPC_question
   generalConfig {defaultMessage=DO NOT TALK TO THE QUEEN. DO NOT BOTHER THE QUEEN.}
  }
 }
 Speech_C_3 {
  id=#11
  class=Speech_C
  outputs [#9]
  properties {
   nodeName=""
   softTaskIcon=&T_UI_NPC_question
   generalConfig {defaultMessage=SHE IS TAKING A LITTLE VACATION AND NEEDS HER REST.}
  }
 }
 Back_C_0 {
  id=#4
  class=Back_C
  sub_nodes [ConversationChoiceNode_7 {class=ConversationChoiceNode, properties {defaultChoiceDisplayText=BACK}}]
 }
 Speech_C_7 {
  id=#3
  class=Speech_C
  outputs [#12]
  properties {nodeName="", softTaskIcon=&T_UI_NPC_question, generalConfig {defaultMessage=ZEUS FELL DOWN HA HA HA.}}
  sub_nodes [ConversationChoiceNode_0 {class=ConversationChoiceNode, properties {defaultChoiceDisplayText=The Gods}}]
 }
 Speech_C_15 {
  id=#12
  class=Speech_C
  outputs [#13]
  properties {
   nodeName=""
   softTaskIcon=&T_UI_NPC_question
   generalConfig {defaultMessage=THE GODS HAVE GONE TO FIND NEW DOMAINS.}
  }
 }
 Speech_C_14 {
  id=#13
  class=Speech_C
  outputs [#9]
  properties {
   nodeName=""
   softTaskIcon=&T_UI_NPC_question
   generalConfig {defaultMessage=MY MASTER LEFT ME HERE TO PREPARE FOR HER ARRIVAL.}
  }
 }
]
